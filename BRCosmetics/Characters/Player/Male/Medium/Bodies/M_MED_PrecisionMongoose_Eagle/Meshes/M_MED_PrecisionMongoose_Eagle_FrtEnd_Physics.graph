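# Physics asset for the FrtEnd mesh of the PrecisionMongoose Eagle male medium body, driving its dangling chain and cloth bones.
# Two chains and four cloth strips (A–D) are capsule strings with limited swing and tight twist, hung from the pelvis through free-rotating CONS and MSTR bodies.
# An angularly locked constraint joins the front and back chain tips into a loop.
# Pelvis and thigh capsules provide leg and hip volumes.
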PhysicsAsset /BRCosmetics/Characters/Player/Male/Medium/Bodies/M_MED_PrecisionMongoose_Eagle/Meshes/M_MED_PrecisionMongoose_Eagle_FrtEnd_Physics
bodies {
 dyn_ChainBack_1 [{shapeType=Capsule, length=4.0342, radius=1.9658, shapeName=dyn_ChainBack_1_capsule, center=<3 0 1.5>, rotation=<rot 90 0 0>}]
 dyn_ChainBack_2 [{
   shapeType=Capsule
   length=3.8441
   radius=2.1559
   shapeName=dyn_ChainBack_2_capsule
   center=<2.5266 0.0315 1.5>
   rotation=<rot 90 180 -176.1953>
  }]
 dyn_ChainBack_3 [{shapeType=Capsule, length=3.9941, radius=2.0059, shapeName=dyn_ChainBack_3_capsule, center=<3 0 1.5>, rotation=<rot 90 0 0>}]
 dyn_ChainBack_4 [{shapeType=Capsule, length=3.9317, radius=2.0683, shapeName=dyn_ChainBack_4_capsule, center=<3 0 1.5>, rotation=<rot 79.2939 0 0.7607>}]
 dyn_ChainBack_CONS [{
   shapeType=Capsule
   length=2.9111
   radius=1.6028
   shapeName=dyn_ChainBack_CONS_capsule
   center=<0.4309 0.1456 -0.1987>
   rotation=<rot 0 2.7762 -89.0703>
  }]
 dyn_ChainBack_MSTR [{shapeType=Capsule, length=1, radius=1, shapeName=dyn_ChainBack_MSTR_capsule}]
 dyn_ChainFor_1 [{shapeType=Capsule, length=4, radius=2, shapeName=dyn_ChainFor_1_capsule, center=<3 0 1.5>, rotation=<rot 90 0 0>}]
 dyn_ChainFor_2 [{shapeType=Capsule, length=3.75, radius=2.25, shapeName=dyn_ChainFor_2_capsule, center=<3 0 1.5>, rotation=<rot 90 0 0>}]
 dyn_ChainFor_3 [{shapeType=Capsule, length=3.75, radius=2.25, shapeName=dyn_ChainFor_3_capsule, center=<3 0 1.5>, rotation=<rot 90 0 0>}]
 dyn_ChainFor_4 [{shapeType=Capsule, length=4.0567, radius=1.9433, shapeName=dyn_ChainFor_4_capsule, center=<3 0.1834 1.5>, rotation=<rot 90 7.125 -0.3107>}]
 dyn_ChainFor_CONS [{
   shapeType=Capsule
   length=2.9242
   radius=1.2724
   shapeName=dyn_ChainFor_CONS_capsule
   center=<0.5613 0.0044 0.2281>
   rotation=<rot 0 -0.2377 -85.1098>
  }]
 dyn_ChainFor_MSTR [{shapeType=Capsule, length=1, radius=1, shapeName=dyn_ChainFor_MSTR_capsule}]
 dyn_ClothForA_1 [{shapeType=Capsule, length=2.584, radius=2.166, shapeName=dyn_ClothForA_1_capsule, center=<6.5 0 1.5>, rotation=<rot 0 0 90>}]
 dyn_ClothForA_2 [{shapeType=Capsule, length=1.7492, radius=3.0008, shapeName=dyn_ClothForA_2_capsule, center=<3 0 2>, rotation=<rot 0 0 90>}]
 dyn_ClothForA_3 [{shapeType=Capsule, length=1.9376, radius=2.8124, shapeName=dyn_ClothForA_3_capsule, center=<2 0 1.5>, rotation=<rot 0.0001 -7.23 90>}]
 dyn_ClothForA_CONS [{
   shapeType=Capsule
   length=2.7385
   radius=1.038
   shapeName=dyn_ClothForA_CONS_capsule
   center=<-0.7509 0.0096 -0.1859>
   rotation=<rot 0 7.483 -89.1084>
  }]
 dyn_ClothForA_MSTR [{shapeType=Capsule, length=1, radius=1, shapeName=dyn_ClothForA_MSTR_capsule}]
 dyn_ClothForB_1 [{shapeType=Capsule, length=2.0448, radius=2.7052, shapeName=dyn_ClothForB_1_capsule, center=<6 0 2>, rotation=<rot 0 0 90>}]
 dyn_ClothForB_2 [{shapeType=Capsule, length=2.3044, radius=2.4456, shapeName=dyn_ClothForB_2_capsule, center=<3 0 1.5>, rotation=<rot 0 0 90>}]
 dyn_ClothForB_3 [{shapeType=Capsule, length=2.4442, radius=2.3058, shapeName=dyn_ClothForB_3_capsule, center=<3 0 1>, rotation=<rot 0 0 90>}]
 dyn_ClothForB_CONS [{
   shapeType=Capsule
   length=2.2467
   radius=0.9528
   shapeName=dyn_ClothForB_CONS_capsule
   center=<-0.4268 0.0983 -0.1446>
   rotation=<rot 0 6.9713 -88.416>
  }]
 dyn_ClothForB_MSTR [{shapeType=Capsule, length=1, radius=1, shapeName=dyn_ClothForB_MSTR_capsule}]
 dyn_ClothForC_1 [{shapeType=Capsule, length=2.125, radius=2.625, shapeName=dyn_ClothForC_1_capsule, center=<4 0 2.3>, rotation=<rot 0 0 90>}]
 dyn_ClothForC_2 [{shapeType=Capsule, length=1.9476, radius=2.8024, shapeName=dyn_ClothForC_2_capsule, center=<3 0 2>, rotation=<rot 0 0 90>}]
 dyn_ClothForC_3 [{shapeType=Capsule, length=1.7418, radius=3.0082, shapeName=dyn_ClothForC_3_capsule, center=<3 0 2>, rotation=<rot 0 0 90>}]
 dyn_ClothForC_CONS [{
   shapeType=Capsule
   length=2.1222
   radius=0.885
   shapeName=dyn_ClothForC_CONS_capsule
   center=<-0.1921 -0.0265 0.1892>
   rotation=<rot 0 6.4836 -89.4583>
  }]
 dyn_ClothForC_MSTR [{shapeType=Capsule, length=1, radius=1, shapeName=dyn_ClothForC_MSTR_capsule}]
 dyn_ClothForD_1 [{shapeType=Capsule, length=2.125, radius=2.625, shapeName=dyn_ClothForD_1_capsule, center=<3.5 0 2.3474>, rotation=<rot 0 0 90>}]
 dyn_ClothForD_2 [{shapeType=Capsule, length=2.228, radius=2.522, shapeName=dyn_ClothForD_2_capsule, center=<3 0 1.5>, rotation=<rot 0 0 90>}]
 dyn_ClothForD_3 [{shapeType=Capsule, length=1.9209, radius=2.8291, shapeName=dyn_ClothForD_3_capsule, center=<3 0 1.5>, rotation=<rot 0 0 90>}]
 dyn_ClothForD_CONS [{
   shapeType=Capsule
   length=2.186
   radius=0.9664
   shapeName=dyn_ClothForD_CONS_capsule
   center=<-0.1276 0.0591 -0.0785>
   rotation=<rot 0 -175.0463 -87.1305>
  }]
 dyn_ClothForD_MSTR [{shapeType=Capsule, length=1, radius=1, shapeName=dyn_ClothForD_MSTR_capsule}]
 dyn_thigh_twist_01_l [{
   shapeType=Capsule
   length=18.4622
   radius=10.1191
   shapeName=thigh_l_capsule
   center=<10.9557 -0.0713 3.4687>
   rotation=<rot 88.5088 9.073 8.7586>
  }]
 pelvis [{shapeType=Capsule, length=7.8059, radius=10.821, shapeName=pelvis_capsule, center=<-1.4081 -1.1074 0.3584>}]
 thigh_l [{
   shapeType=Capsule
   length=12.2281
   radius=9.2036
   shapeName=thigh_l_capsule
   center=<-28.4059 -0.0884 -0.1002>
   rotation=<rot 87.0783 175.8665 170.0926>
  }]
 thigh_twist_01_r [{
   shapeType=Capsule
   length=20.2848
   radius=10.2933
   shapeName=thigh_l_capsule
   center=<-12.3141 0.6861 -2.1016>
   rotation=<rot 89.6785 -92.7196 89.1553>
  }]
}
constraints [
 {
  bone1Name=dyn_ChainBack_MSTR
  bone2Name=dyn_ChainBack_CONS
  swing1LimitDegrees=45
  swing1Motion=Free
  swing2LimitDegrees=45
  swing2Motion=Free
  twistLimitDegrees=45
  twistMotion=Free
 }
 {bone1Name=dyn_ChainBack_1, bone2Name=dyn_ChainBack_MSTR, swing1LimitDegrees=30, swing2LimitDegrees=30, twistLimitDegrees=2}
 {bone1Name=dyn_ChainBack_2, bone2Name=dyn_ChainBack_1, swing1LimitDegrees=30, swing2LimitDegrees=30, twistLimitDegrees=2}
 {bone1Name=dyn_ChainBack_3, bone2Name=dyn_ChainBack_2, swing1LimitDegrees=30, swing2LimitDegrees=30, twistLimitDegrees=2}
 {bone1Name=dyn_ChainBack_4, bone2Name=dyn_ChainBack_3, swing1LimitDegrees=15, swing2LimitDegrees=15, twistLimitDegrees=2}
 {
  bone1Name=dyn_ChainFor_MSTR
  bone2Name=dyn_ChainFor_CONS
  swing1LimitDegrees=45
  swing1Motion=Free
  swing2LimitDegrees=45
  swing2Motion=Free
  twistLimitDegrees=45
  twistMotion=Free
 }
 {bone1Name=dyn_ChainFor_1, bone2Name=dyn_ChainFor_MSTR, swing1LimitDegrees=30, swing2LimitDegrees=20, twistLimitDegrees=2}
 {bone1Name=dyn_ChainFor_2, bone2Name=dyn_ChainFor_1, swing1LimitDegrees=30, swing2LimitDegrees=20, twistLimitDegrees=2}
 {bone1Name=dyn_ChainFor_3, bone2Name=dyn_ChainFor_2, swing1LimitDegrees=30, swing2LimitDegrees=30, twistLimitDegrees=2}
 {bone1Name=dyn_ChainFor_4, bone2Name=dyn_ChainFor_3, swing1LimitDegrees=15, swing2LimitDegrees=15, twistLimitDegrees=2}
 {
  bone1Name=dyn_ClothForA_MSTR
  bone2Name=dyn_ClothForA_CONS
  swing1LimitDegrees=45
  swing1Motion=Free
  swing2LimitDegrees=45
  swing2Motion=Free
  twistLimitDegrees=45
  twistMotion=Free
 }
 {bone1Name=dyn_ClothForA_1, bone2Name=dyn_ClothForA_MSTR, swing1LimitDegrees=30, swing2LimitDegrees=60, twistLimitDegrees=2}
 {bone1Name=dyn_ClothForA_2, bone2Name=dyn_ClothForA_1, swing1LimitDegrees=30, swing2LimitDegrees=60, twistLimitDegrees=2}
 {bone1Name=dyn_ClothForA_3, bone2Name=dyn_ClothForA_2, swing1LimitDegrees=30, swing2LimitDegrees=60, twistLimitDegrees=2}
 {
  bone1Name=dyn_ClothForB_MSTR
  bone2Name=dyn_ClothForB_CONS
  swing1LimitDegrees=45
  swing1Motion=Free
  swing2LimitDegrees=45
  swing2Motion=Free
  twistLimitDegrees=45
  twistMotion=Free
 }
 {bone1Name=dyn_ClothForB_1, bone2Name=dyn_ClothForB_MSTR, swing1LimitDegrees=30, swing2LimitDegrees=45, twistLimitDegrees=2}
 {bone1Name=dyn_ClothForB_2, bone2Name=dyn_ClothForB_1, swing1LimitDegrees=30, swing2LimitDegrees=45, twistLimitDegrees=2}
 {bone1Name=dyn_ClothForB_3, bone2Name=dyn_ClothForB_2, swing1LimitDegrees=30, swing2LimitDegrees=45, twistLimitDegrees=2}
 {
  bone1Name=dyn_ClothForC_MSTR
  bone2Name=dyn_ClothForC_CONS
  swing1LimitDegrees=45
  swing1Motion=Free
  swing2LimitDegrees=45
  swing2Motion=Free
  twistLimitDegrees=45
  twistMotion=Free
 }
 {bone1Name=dyn_ClothForC_1, bone2Name=dyn_ClothForC_MSTR, swing1LimitDegrees=30, swing2LimitDegrees=45, twistLimitDegrees=2}
 {bone1Name=dyn_ClothForC_2, bone2Name=dyn_ClothForC_1, swing1LimitDegrees=30, swing2LimitDegrees=45, twistLimitDegrees=2}
 {bone1Name=dyn_ClothForC_3, bone2Name=dyn_ClothForC_2, swing1LimitDegrees=30, swing2LimitDegrees=45, twistLimitDegrees=2}
 {
  bone1Name=dyn_ClothForD_MSTR
  bone2Name=dyn_ClothForD_CONS
  swing1LimitDegrees=45
  swing1Motion=Free
  swing2LimitDegrees=45
  swing2Motion=Free
  twistLimitDegrees=45
  twistMotion=Free
 }
 {bone1Name=dyn_ClothForD_1, bone2Name=dyn_ClothForD_MSTR, swing1LimitDegrees=30, swing2LimitDegrees=45, twistLimitDegrees=2}
 {bone1Name=dyn_ClothForD_2, bone2Name=dyn_ClothForD_1, swing1LimitDegrees=30, swing2LimitDegrees=45, twistLimitDegrees=2}
 {bone1Name=dyn_ClothForD_3, bone2Name=dyn_ClothForD_2, swing1LimitDegrees=30, swing2LimitDegrees=45, twistLimitDegrees=2}
 {
  bone1Name=thigh_l
  bone2Name=pelvis
  swing1LimitDegrees=45
  swing1Motion=Free
  swing2LimitDegrees=45
  swing2Motion=Free
  twistLimitDegrees=45
  twistMotion=Free
 }
 {
  bone1Name=dyn_ChainBack_CONS
  bone2Name=pelvis
  swing1LimitDegrees=45
  swing1Motion=Free
  swing2LimitDegrees=45
  swing2Motion=Free
  twistLimitDegrees=45
  twistMotion=Free
 }
 {
  bone1Name=dyn_ChainFor_CONS
  bone2Name=pelvis
  swing1LimitDegrees=45
  swing1Motion=Free
  swing2LimitDegrees=45
  swing2Motion=Free
  twistLimitDegrees=45
  twistMotion=Free
 }
 {
  bone1Name=dyn_ClothForA_CONS
  bone2Name=pelvis
  swing1LimitDegrees=45
  swing1Motion=Free
  swing2LimitDegrees=45
  swing2Motion=Free
  twistLimitDegrees=45
  twistMotion=Free
 }
 {
  bone1Name=dyn_ClothForB_CONS
  bone2Name=pelvis
  swing1LimitDegrees=45
  swing1Motion=Free
  swing2LimitDegrees=45
  swing2Motion=Free
  twistLimitDegrees=45
  twistMotion=Free
 }
 {
  bone1Name=dyn_ClothForC_CONS
  bone2Name=pelvis
  swing1LimitDegrees=45
  swing1Motion=Free
  swing2LimitDegrees=45
  swing2Motion=Free
  twistLimitDegrees=45
  twistMotion=Free
 }
 {
  bone1Name=dyn_ClothForD_CONS
  bone2Name=pelvis
  swing1LimitDegrees=45
  swing1Motion=Free
  swing2LimitDegrees=45
  swing2Motion=Free
  twistLimitDegrees=45
  twistMotion=Free
 }
 {
  bone1Name=dyn_thigh_twist_01_l
  bone2Name=thigh_l
  swing1LimitDegrees=45
  swing1Motion=Free
  swing2LimitDegrees=45
  swing2Motion=Free
  twistLimitDegrees=45
  twistMotion=Free
 }
 {
  bone1Name=dyn_ChainFor_4
  bone2Name=dyn_ChainBack_4
  swing1LimitDegrees=30
  swing1Motion=Locked
  swing2LimitDegrees=30
  swing2Motion=Locked
  twistLimitDegrees=2
  twistMotion=Locked
 }
 {
  bone1Name=thigh_twist_01_r
  bone2Name=pelvis
  swing1LimitDegrees=45
  swing1Motion=Free
  swing2LimitDegrees=45
  swing2Motion=Free
  twistLimitDegrees=45
  twistMotion=Free
 }
]
